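# UI material for the special shop background that tiles and pans a pattern texture with a perspective warp.
# UVs are scaled around a point, run through MF_UI_BasicPerspective, repeated by PatternRepeat, then panned by Speed.
# A rotated, offset linear gradient blends between the pattern and its inverse.
# Scale, offset and speed read only one channel for both axes, the pixel-size ratio divides R by itself, and InvertPattern is unconnected.
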Material /ItemShop/UI/Art/SpecialShopBG/M_UI_SpecialShopBG_Pan
graph {
 properties {bUsedWithStaticMesh=false, materialDomain=MD_UI, pixelDepthOffsetMode=PDOM_Legacy, usedShadingModels=Default Lit}
 nodes [
  n1 = VectorParameter() {parameterName=PerpScale, channelNames {a=OffsetY, b=OffsetX, g=Y, r=X}, defaultValue=<rgba 1 1 0 0>}
  n2 = Constant() {r=1}
  n3 = MaterialFunctionCall!(
   UV=MaterialFunctionCall!(
    Texture Scale=AppendVector(A=@n1.X, B=@n1.X)
    UVs=TextureCoordinate()
    Offset XY Texturesize Z=MaterialFunctionCall(R=@n1.OffsetX, G=@n1.OffsetX, B=@n2) {materialFunction=&AppendMany}.RGB
   ) {materialFunction=/Engine/Functions/Engine_MaterialFunctions01/Texturing/ScaleUVsAroundPoint}.UVs
   Texture
   FixedScaleFactor_X
   FixedScaleFactor_Y
   Persp_Y=ScalarParameter() {group=Pattern, parameterName=Persp_Y}
   Persp_X=ScalarParameter() {group=Pattern, parameterName=Persp_X}
   Slide_X=ScalarParameter() {group=Pattern, parameterName=Slide_X}
   Slide_Y=ScalarParameter() {group=Pattern, parameterName=Slide_Y}
   Flip_Y
   Flip_X
  ) {materialFunction=/Game/UI/Foundation/Materials/MF_UI_BasicPerspective}
  n4 = MaterialFunctionCall(Float2=MaterialFunctionCall() {materialFunction=&GetUserInterfaceUV}.Pixel Size) {materialFunction=&BreakOutFloat2Components}
  n5 = Divide(A=@n4.R, B=@n4.R)
  n6 = Constant() {r=1}
  n7 = NamedRerouteDeclaration(None=Multiply(A=Add(A=@n3.FinalUvs) {constB=0}, B=AppendVector(A=@n5, B=@n6))) {name=NormalizedSSUVs, nodeColor=<rgba 1 0 0.7969 1>}
  n8 = ScalarParameter() {defaultValue=4, parameterName=PatternRepeat}
  n9 = VectorParameter() {parameterName=Speed, channelNames {a="", b="", g=Y, r=X}, defaultValue=<rgba 0 0 0 1>}
  n10 = TextureSampleParameter2D(UVs=Panner(Coordinate=Multiply(A=@n7, B=@n8), Speed=AppendVector(A=@n9.X, B=@n9.X))) {parameterName=Pattern, texture=/OfferCatalog/Art/SpecialShopBG/T_UI_ShopBG_BCG}
  n11 = MaterialFunctionCall(UVs=TextureCoordinate(), "Rotation Angle (0-1)"=ScalarParameter() {defaultValue=0.25, parameterName=InvertAngle}) {materialFunction=&CustomRotator}
  n12 = Add(
   A=Multiply(A=ComponentMask(None=@n11.Rotated Values) {g=true}, B=ScalarParameter() {parameterName=InvertAngleMuilt})
   B=ScalarParameter() {parameterName=InvertOffset}
  )
  LinearInterpolate(
   A=OneMinus(None=@n10.RGB)
   B=@n10.RGB
   Alpha=OneMinus(None=TextureSample(UVs=@n12) {samplerType=SAMPLERTYPE_LinearColor, texture=/Game/Global/Textures/Default/Masks_and_Gradients/linear_gradient}.R)
  )
  ScalarParameter() {parameterName=InvertPattern, sliderMax=1}
 ]
}
